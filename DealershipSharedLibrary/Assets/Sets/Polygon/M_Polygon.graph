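# M_Polygon is an unlit, translucent, two-sided Niagara particle material for the dealership library's Polygon set.
# A Switch_Selector scalar drives color and opacity switches between color passthrough, main geometry and triangle looks, with color passing through MF_FX_EmissiveAdjustments.
# Main geometry uses UV sphere masks for a boosted color gradient and alpha cutout.
# Triangles blend particle and secondary colors through an edge texture, varied per particle by a stepped random value.
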Material /DealershipSharedLibrary/Assets/Sets/Polygon/M_Polygon
graph {
 properties {
  bAutomaticallySetUsageInEditor=false
  bUsedWithNiagaraMeshParticles=true
  bUsedWithNiagaraRibbons=true
  bUsedWithNiagaraSprites=true
  blendMode=BLEND_Translucent
  shadingModel=MSM_Unlit
  twoSided=true
  usedShadingModels=Unlit
 }
 nodes [
  n1 = NamedRerouteUsage()
  MaterialExpressionNamedRerouteUsage_4 = NamedRerouteUsage()
  MaterialExpressionNamedRerouteUsage_6 = NamedRerouteUsage()
  MaterialExpressionNamedRerouteUsage_1 = NamedRerouteUsage()
  n2 = Switch!(
   SwitchValue=@n1
   Default
   00-ColorPassthrough=@MaterialExpressionNamedRerouteUsage_4
   01-MainGeo=@MaterialExpressionNamedRerouteUsage_6
   02-Triangles=@MaterialExpressionNamedRerouteUsage_1
  ) {
   description=SwitchRGB
   inputs [
    {
     inputName=00-ColorPassthrough
     input {
      expression=&self:MaterialExpressionNamedRerouteUsage_4
      inputName=None
      mask=0
      maskA=0
      maskB=0
      maskG=0
      maskR=0
      outputIndex=0
     }
    }
    {
     inputName=01-MainGeo
     input {
      expression=&self:MaterialExpressionNamedRerouteUsage_6
      inputName=None
      mask=0
      maskA=0
      maskB=0
      maskG=0
      maskR=0
      outputIndex=0
     }
    }
    {
     inputName=02-Triangles
     input {
      expression=&self:MaterialExpressionNamedRerouteUsage_1
      inputName=None
      mask=0
      maskA=0
      maskB=0
      maskG=0
      maskR=0
      outputIndex=0
     }
    }
   ]
  }
  MaterialFunctionCall(Emissive=@n2) {materialFunction=&MF_FX_EmissiveAdjustments}
  NamedRerouteDeclaration(None=VectorParameter() {group=00_Main, parameterName=SecondaryColor, sortPriority=20, defaultValue=<rgba 0 0 0 1>}.RGB) {name=Color2, nodeColor=<rgba 0.4678 0.2747 0.0908 1>}
  n3 = ParticleColor()
  NamedRerouteDeclaration(None=@n3.RGB) {name=Color1, nodeColor=<rgba 0.4678 0.2747 0.0908 1>}
  NamedRerouteDeclaration(None=DynamicParameter() {defaultValue=<rgba 0 1 1 1>, paramNames [Animation1,Param2,Param3,Param4]}.Animation1) {name=Dinamic1, nodeColor=<rgba 0.4678 0.2747 0.0908 1>}
  NamedRerouteDeclaration(None=@n3.A) {name=Alpha1, nodeColor=<rgba 0.4678 0.2747 0.0908 1>}
  n4 = NamedRerouteUsage()
  n5 = TextureCoordinate() {vTiling=0}
  n6 = Constant()
  NamedRerouteDeclaration(None=LinearInterpolate(A=Multiply(A=@n4) {constB=100}, B=NamedRerouteUsage(), Alpha=SphereMask(A=@n5, B=@n6) {attenuationRadius=0.5})) {name=MainGeoRGB, nodeColor=<rgba 0.0016 0.5378 1 1>}
  n7 = NamedRerouteUsage()
  n8 = NamedRerouteUsage()
  n9 = TextureSample() {texture=/DealershipSharedLibrary/FX/Textures/Masks/T_Triangle_Edges}
  n10 = ParticleRandom()
  NamedRerouteDeclaration(None=LinearInterpolate(A=@n7, B=LinearInterpolate(A=@n8, B=@n7, Alpha=@n9.R), Alpha=Step(Y=@n10) {constX=0.5})) {name=TrianglesRGB, nodeColor=<rgba 0.2793 1 0 1>}
  n11 = SphereMask(A=TextureCoordinate() {uTiling=0}, B=@n6, Radius=NamedRerouteUsage()) {attenuationRadius=0.5}
  NamedRerouteDeclaration(None=Multiply(A=NamedRerouteUsage(), B=OneMinus(None=@n11))) {name=MainGeoA, nodeColor=<rgba 0.0276 0.8455 1 1>}
  NamedRerouteDeclaration(None=ScalarParameter() {group=00_Main, parameterName=Switch_Selector, sortPriority=10}) {name=Switch_Selector, nodeColor=<rgba 0.4678 0.2747 0.0908 1>}
  MaterialExpressionNamedRerouteUsage_5 = NamedRerouteUsage()
  MaterialExpressionNamedRerouteUsage_8 = NamedRerouteUsage()
  Switch(
   SwitchValue=@n1
   00-ColorPassthrough=@MaterialExpressionNamedRerouteUsage_5
   01-MainGeo=@MaterialExpressionNamedRerouteUsage_8
   02-Triangles=@MaterialExpressionNamedRerouteUsage_5
  ) {
   description=SwitchA
   inputs [
    {
     inputName=00-ColorPassthrough
     input {
      expression=&self:MaterialExpressionNamedRerouteUsage_5
      inputName=None
      mask=0
      maskA=0
      maskB=0
      maskG=0
      maskR=0
      outputIndex=0
     }
    }
    {
     inputName=01-MainGeo
     input {
      expression=&self:MaterialExpressionNamedRerouteUsage_8
      inputName=None
      mask=0
      maskA=0
      maskB=0
      maskG=0
      maskR=0
      outputIndex=0
     }
    }
    {
     inputName=02-Triangles
     input {
      expression=&self:MaterialExpressionNamedRerouteUsage_5
      inputName=None
      mask=0
      maskA=0
      maskB=0
      maskG=0
      maskR=0
      outputIndex=0
     }
    }
   ]
  }
 ]
}
